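# Conversation database for the BeefBoss NPC, entered through the Conversation.Entry.BeefBoss tag.
# A single speech node gives a default greeting, with priority-weighted contextual lines for players wearing certain Taco or Pizza outfits or owning the Nightmare dialogue-override quest tag.
# The speech node then links out by entry tag to the shared NPC Library quests and basic services conversations.
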ConversationDatabase /NPCLibrary/NPCs/BeefBoss/BeefBoss/NPCLibrary_Conversation_BeefBoss
entry_points [{tag=Conversation.Entry.BeefBoss, destinations [#1]}]
nodes [
 ConversationEntryPointNode_7 {
  id=#1
  class=ConversationEntryPointNode
  outputs [#2]
  properties {entryTag {tagName=Conversation.Entry.BeefBoss}}
 }
 Speech_C_4 {
  id=#2
  class=Speech_C
  outputs [#3, #4]
  properties {
   nodeName=""
   generalConfig {
    defaultMessage="Have a beefy day!"
    contextualMessages [
     {
      message=Look who it is....
      requirementMatchPolicy=RequireAny
      contextRequirements [{requirement {
         class=FortControllerRequirement_HasCID
         properties {allowedCIds [
           /BRCosmetics/Athena/Items/Cosmetics/Characters/CID_538_Athena_Commando_M_Taco
           /BRCosmetics/Athena/Items/Cosmetics/Characters/CID_797_Athena_Commando_F_Taco
           /BRCosmetics/Athena/Items/Cosmetics/Characters/CID_109_Athena_Commando_M_Pizza
           /BRCosmetics/Athena/Items/Cosmetics/Characters/CID_831_Athena_Commando_F_PIzzaPitMascot
          ]}
        }}]
      priority {curve {rowName=Tandem.Speech.Priority.OutfitReaction}}
     }
     {
      message="I was just flippin' some patties when some sort of... laser-sight showed up on my grill. Terrifying!"
      contextRequirements [{requirement {class=FortControllerRequirement_HasOwnedTag, properties {ownedTag {tagName=Athena.Quests.NPC.DialogueOverride.Nightmare}}}}]
      priority {curve {rowName=Tandem.Speech.Priority.HasQuestGameplayTag}}
     }
    ]
   }
  }
 }
 ConversationLinkNode_0 {id=#4, class=ConversationLinkNode, properties {remoteEntryTag {tagName=Conversation.Entry.BasicServices}}}
 ConversationLinkNode_1 {id=#3, class=ConversationLinkNode, properties {remoteEntryTag {tagName=Conversation.Entry.NPCLibrary.Quests}}}
]
